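# Anim Blueprint for the Slurp Monster pickaxe that cycles the mesh through axe, pickaxe and blob morph targets over time.
# Each update advances a looping 0–100 counter, SpinX, by delta time and SpinPerSecond.
# SpinX windows set eased FInterpTo weights for the axe, pickaxe and one shared blob weight, which SpinX routes to one of three blob targets.
# The AnimGraph passes its input pose through unchanged.
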AnimBlueprint /BRCosmetics/Weapons/FORT_Melee/Pickaxe_Slurp_Monster/Meshes/Slurp_Monster_Axe_AnimBlueprint
target_skeleton=&/Slurp_Monster_Axe_Skeleton
graphs [
 AnimGraph {nodes [
   AnimGraphNode_SubInput_0 = Input Pose()
   Output Pose(Result=@AnimGraphNode_SubInput_0) {layerGroup=None}
  ]}
 EventGraph {nodes [
   n1 = Event BlueprintUpdateAnimation()
   > Set time(time:double=@n1.DeltaTimeX) -> (Output_Get:double)
   > n2 = Sequence!(execute) -> (then_0, then_1, then_2:Exec, then_3:Exec, then_4:Exec)
   > n3 = Branch(Condition="float >= float"(A=Get SpinX() -> (SpinX:double), B=100) {is=title K2Node_CallFunction})
   > Set SpinX(SpinX:double=float - float(A=Get SpinX(), B=100) {is=title K2Node_CallFunction}) -> (Output_Get:double)
   n4 = Get time() -> (time:double)
   n5 = Get SpinPerSecond() -> (SpinPerSecond:double)
   Set SpinX(execute=@n3.else, SpinX=float + float(A=Get SpinX(), B=float * float(A=@n4, B=@n5)) {is=title K2Node_CommutativeAssociativeBinaryOperator})
   SetMorphTarget!(execute=@n2.then_1, self:Anim Instance Object Reference, MorphTargetName:Name=axe, Value:float=Get axeMorphDelta() -> (axeMorphDelta:double)) -> (then)
   > SetMorphTarget(MorphTargetName=pickaxe, Value=Get pickaxeMorphDelta() -> (pickaxeMorphDelta:double))
   > n6 = Branch(Condition="float > float"(A=Get SpinX(), B=30) {is=title K2Node_CallFunction})
   > n7 = Branch(Condition="float > float"(A=60, B=Get SpinX()))
   > SetMorphTarget(MorphTargetName=Blob_03, Value=Get Blob 01 Delta() -> (Blob 01 Delta:double))
   SetMorphTarget(execute=@n7.else, MorphTargetName=Blob02, Value=Get Blob 01 Delta())
   SetMorphTarget(execute=@n6.else, MorphTargetName=Blob01, Value=Get Blob 01 Delta())
   n8 = Get SpinX()
   n9 = Set PickAxeMorphActive(
    execute=@n2.then_3
    PickAxeMorphActive:bool=AND Boolean(A="float > float"(A=@n8, B=74.9), B="float < float"!(A:double=@n8, B:double=100) -> (ReturnValue:bool) {is=title K2Node_CallFunction})
   ) -> (Output_Get:bool)
   > n10 = Branch(Condition=@n9)
   > Set pickaxeMorphDelta(pickaxeMorphDelta:double=FInterpTo(Current=Get pickaxeMorphDelta(), Target=1, DeltaTime=Get time(), InterpSpeed=20)) -> (Output_Get:double)
   Set pickaxeMorphDelta(execute=@n10.else, pickaxeMorphDelta=FInterpTo(Current=Get pickaxeMorphDelta(), Target=0, DeltaTime=Get time(), InterpSpeed=20))
   n11 = Get SpinX()
   n12 = Set axeMorphActive(execute=@n2.then_2, axeMorphActive:bool=AND Boolean(A="float > float"(A=@n11, B=41.6), B="float < float"(A=@n11, B=66.6))) -> (Output_Get:bool)
   > n13 = Branch(Condition=@n12)
   > Set axeMorphDelta(axeMorphDelta:double=FInterpTo(Current=Get axeMorphDelta(), Target=1, DeltaTime=Get time(), InterpSpeed=20)) -> (Output_Get:double)
   Set axeMorphDelta(execute=@n13.else, axeMorphDelta=FInterpTo(Current=Get axeMorphDelta(), Target=0, DeltaTime=Get time(), InterpSpeed=20))
   n14 = Get SpinX()
   n15 = "float > float"(A=@n14, B=0)
   n16 = "float < float"(A=@n14, B=8.3)
   n17 = Get SpinX()
   n18 = "float > float"(A=@n17, B=33.3)
   n19 = "float < float"(A=@n17, B=41.6)
   n20 = Get SpinX()
   n21 = "float > float"(A=@n20, B=66.6)
   n22 = "float < float"(A=@n20, B=74.9)
   n23 = Branch(
    execute=@n2.then_4
    Condition=OR Boolean!(A=AND Boolean(A=@n15, B=@n16), B=AND Boolean(A=@n18, B=@n19), C:bool=AND Boolean(A=@n21, B=@n22)) -> (ReturnValue)
   )
   > Set Blob 01 Delta(Blob 01 Delta:double=FInterpTo(Current=Get Blob 01 Delta(), Target=1, DeltaTime=Get time(), InterpSpeed=10)) -> (Output_Get:double)
   Set Blob 01 Delta(execute=@n23.else, Blob 01 Delta=FInterpTo(Current=Get Blob 01 Delta(), Target=0, DeltaTime=Get time(), InterpSpeed=10))
   TryGetPawnOwner()
  ]}
]
variables [
 axeMorphActive {type=boolean}
 axeMorphDelta {type=number}
 Blob 01 Delta {type=number}
 PickAxeMorphActive {type=boolean}
 pickaxeMorphDelta {type=number}
 SpinPerSecond {type=number, default=10}
 SpinX {type=number}
 time {type=number}
]
